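# Unlit animated material for a Sparks screen showing a glitching vine pattern.
# The vine texture scrolls with AnimRamp-driven UV distortion and is sampled three times, two with sine-driven offsets, for an RGB split.
# It adds over a dim tile background warped by scanlines and digital noise, then passes through the shared screen-details function.
# Named reroutes define stage colors, beat-driven reactivity, glitch time and random flicker.
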Material /SparksMaps/Effects/Screens/GlitchVine/M_Sparks_Screen_GlitchVine_V2
graph {
 properties {
  bAutomaticallySetUsageInEditor=false
  bHasPixelAnimation=true
  bUsedWithNanite=true
  pixelDepthOffsetMode=PDOM_Legacy
  shadingModel=MSM_Unlit
  usedShadingModels=Unlit
 }
 nodes [
  n1 = Power(Base=MaterialFunctionCall() {materialFunction=&LinearGradient}.VGradient) {constExponent=3}
  n2 = NamedRerouteUsage()
  n3 = NamedRerouteUsage()
  n4 = Time()
  n5 = If(A=NamedRerouteUsage(), "A > B"=Constant() {r=0.02}, "A < B"=Constant() {r=0.5})
  n6 = Saturate(None=MaterialFunctionCall!(
    Number of Scanlines=Constant() {r=5}
    Scanline Speed
    "Scanlines Thickness (Bias)"
    Scanlines Dark Value
    Scanlines Light Value
    Block Opacity=Constant() {r=5}
    UVs=TextureCoordinate()
    Time=Multiply(A=@n4, B=@n5)
    Block Repeat=Constant() {r=0.1}
   ) {materialFunction=/Game/Effects/Fort_Effects/Effects/Atlas/Materials/MF_Atlas_Scanlines_01}.Result)
  n7 = Multiply(
   A=TextureCoordinate() {uTiling=0.5, vTiling=2}
   B=AppendVector(A=ScalarParameter() {defaultValue=10, parameterName=Param_3}, B=ScalarParameter() {defaultValue=0.1, parameterName=Param_4})
  )
  n8 = Multiply(A=TextureSample(UVs=Panner(Coordinate=@n7) {speedY=0.1}) {samplerType=SAMPLERTYPE_LinearColor, texture=/Game/Global/Textures/Default/Noise/T_Atlas_DigitalNoise_Small}.R) {constB=0.5}
  n9 = TextureSample(UVs=LinearInterpolate(A=TextureCoordinate(), B=Add(A=@n6, B=@n8), Alpha=ScalarParameter() {defaultValue=0.25, parameterName=Param_5}) {constAlpha=0.25}) {samplerType=SAMPLERTYPE_LinearColor, texture=/Game/Effects/Fort_Effects/FrontEnd/S33/GateContent/Tile_output_8}
  n10 = Add(A=@n6) {constB=0.25}
  n11 = Multiply(A=Multiply(A=Multiply(A=@n3, B=@n9.R), B=Saturate(None=@n10)), B=ScalarParameter() {defaultValue=0.05}) {constB=0.1}
  Add(A=Multiply(A=Multiply(A=@n1, B=@n2)) {constB=0.1}, B=@n11)
  Add()
  Constant() {r=-0.01}
  Constant() {r=0.01}
  Constant() {r=1}
  n12 = TextureCoordinate()
  n13 = Multiply(
   A=@n12
   B=AppendVector(A=ScalarParameter() {defaultValue=0.75, parameterName=Param_6}, B=ScalarParameter() {defaultValue=0.6, parameterName=Param_7})
  )
  n14 = ScalarParameter() {defaultValue=0.15, parameterName=Param_2}
  n15 = Frac(None=Multiply(A=Time()) {constB=0.025})
  n16 = Constant()
  n17 = TextureCoordinate()
  n18 = NamedRerouteUsage()
  n19 = Constant() {r=0.1}
  n20 = Constant() {r=0.25}
  n21 = TextureSample(UVs=Panner(
    Coordinate=MaterialFunctionCall(XY=@n17) {materialFunction=&Swizzle}.YX
    Speed=If(A=@n18, "A > B"=@n19, "A < B"=@n20)
   ) {speedX=0.1}) {samplerType=SAMPLERTYPE_LinearGrayscale, texture=/Game/Global/Textures/Default/Masks_and_Gradients/AnimRamp}
  n22 = NamedRerouteUsage()
  n23 = Constant() {r=0.005}
  n24 = Constant() {r=0.05}
  n25 = Add(
   A=Add(A=Add(A=@n13, B=@n14), B=AppendVector(A=@n15, B=@n16))
   B=Multiply(A=SmoothStep(Value=@n21.R) {constMax=0.35, constMin=0.3}, B=If(A=@n22, "A > B"=@n23, "A < B"=@n24)) {constB=0.02}
  )
  n26 = TextureSample(UVs=@n25) {samplerType=SAMPLERTYPE_LinearColor, texture=&/T_Sparks_Screen_GlitchVine}
  n27 = Add(A=@n26.R, B=@n26.R)
  n28 = Time()
  n29 = Clamp(None=Sine(None=Multiply(A=@n28) {constB=1.76})) {maxDefault=0.004, minDefault=-0.005}
  n30 = Constant() {r=-0.004}
  n31 = TextureSample(UVs=Add(A=@n25, B=AppendVector(A=@n29, B=@n30))) {samplerType=SAMPLERTYPE_LinearColor, texture=&T_Sparks_Screen_GlitchVine}
  n32 = Add(A=@n31.R, B=@n31.R)
  n33 = Time()
  n34 = Clamp(None=Sine(None=Multiply(A=@n33) {constB=2.43})) {maxDefault=0.005, minDefault=-0.006}
  n35 = Constant() {r=0.005}
  n36 = TextureSample(UVs=Add(A=@n25, B=AppendVector(A=@n34, B=@n35))) {samplerType=SAMPLERTYPE_LinearColor, texture=&T_Sparks_Screen_GlitchVine}
  n37 = Add(A=@n36.R, B=@n36.R)
  n38 = MaterialFunctionCall(R=Power(Base=Add(A=@n27, B=@n26.B)), G=Power(Base=Add(A=@n32, B=@n31.B)), B=Power(Base=Add(A=@n37, B=@n36.B))) {materialFunction=&AppendMany}
  n39 = ScalarParameter() {defaultValue=0.25, parameterName=Param_1}
  MaterialFunctionCall!(Emissive=Add(A=NamedRerouteDeclaration(None=@n11) {name=BG, nodeColor=<rgba 0.2656 1 0 1>}, B=Multiply(A=@n38.RGB, B=@n39)), Scanlines Tile) {materialFunction=/SparksMaps/Effects/Screens/Materials/MF/MF_Sparks_Screen_Details}
  n40 = NamedRerouteUsage()
  Multiply(A=Saturate(None=Add(A=@n40) {constB=0.1}), B=@n38.RGB)
  Multiply(A=NamedRerouteUsage()) {constB=0.1}
  n41 = MaterialFunctionCall!(CustomUPosition) {materialFunction=/SparksStageShow/Systems/StageManager/FXManager/MaterialFunctions/MF_MojoStageColors}
  NamedRerouteDeclaration(None=@n41.Primary) {name=Primary Color, nodeColor=<rgba 0 0 0 1>}
  n42 = CurveAtlasRowParameter(CurveTime=CollectionParameter() {collection=/FMAudioReactiveData/AudioAnalysis/MPC_Sparks_Audio_Parameters, parameterName=BeatLifetime}) {
   atlas=/SparksMaps/Effects/Screens/Curves/CurveAtlas_Audio_Beat_A
   curve=/SparksMaps/Effects/Screens/Curves/Curve_Audio_Beat_Remap_001
   parameterName=ReactivityCurve
  }
  NamedRerouteDeclaration(None=MaterialFunctionCall(Input=@n42, Input Low=Constant(), Input High=Constant() {r=1}, Target Low=Constant() {r=0.75}, Target High=Constant() {r=2}) {materialFunction=&RemapValueRange}.Result) {name=Reactivity, nodeColor=<rgba 0 0 0 1>}
  n43 = ScalarParameter() {defaultValue=1, parameterName=Glitch Anim Amount}
  n44 = Multiply(A=Multiply(A=Time()) {constB=0.5}, B=@n43)
  n45 = Multiply(A=Multiply(A=Time()) {constB=0.376}, B=@n43)
  n46 = Multiply(A=Multiply(A=Time()) {constB=0.4}, B=@n43)
  NamedRerouteDeclaration(None=LinearInterpolate(A=Sine(None=@n44), B=Sine(None=@n45), Alpha=Sine(None=@n46))) {name=Glitch Time, nodeColor=<rgba 0 0.1094 1 1>}
  NamedRerouteDeclaration(None=@n41.Secondary) {name=Secondary Color, nodeColor=<rgba 0 0 0 1>}
  NamedRerouteDeclaration(None=Constant() {r=0.25}) {name=GlitchFrequency, nodeColor=<rgba 0.0547 1 0 1>}
  NamedRerouteDeclaration(None=Constant() {r=0.5}) {name=GlitchFrequencyMultiplier, nodeColor=<rgba 0.0547 1 0 1>}
  NamedRerouteDeclaration(None=VectorParameter() {group=ScreenThemeParams, parameterName=TestColorB, defaultValue=<rgba 1 0 0.4717 1>}.RGB) {name=ColorB, nodeColor=<rgba 0.5938 0 1 1>}
  NamedRerouteDeclaration(None=VectorParameter() {group=ScreenThemeParams, parameterName=TestColorA, defaultValue=<rgba 1 0 0.4558 1>}.RGB) {name=ColorA, nodeColor=<rgba 0 1 0.2734 1>}
  NamedRerouteDeclaration(None=Constant() {r=5}) {name=TimeMult, nodeColor=<rgba 1 0.2109 0 1>}
  NamedRerouteDeclaration(None=Constant() {r=1}) {name=LerpMult, nodeColor=<rgba 0 0.1094 1 1>}
  n47 = NamedRerouteUsage()
  n48 = Time()
  n49 = ScalarParameter() {defaultValue=0.125, group=RandomFloatFlicker, parameterName=RandomFlicker_Value01}
  n50 = Sine(None=Multiply(A=NamedRerouteUsage(), B=Multiply(A=@n48, B=@n49)))
  n51 = Time()
  n52 = ScalarParameter() {defaultValue=0.134, group=RandomFloatFlicker, parameterName=RandomFlicker_Value02}
  n53 = Sine(None=Multiply(A=NamedRerouteUsage(), B=Multiply(A=@n51, B=@n52)))
  n54 = Time()
  n55 = ScalarParameter() {defaultValue=0.23, group=RandomFloatFlicker, parameterName=RandomFlicker_Value03}
  n56 = Sine(None=Multiply(A=NamedRerouteUsage(), B=Multiply(A=@n54, B=@n55)))
  n57 = Time()
  n58 = ScalarParameter() {defaultValue=0.185, group=RandomFloatFlicker, parameterName=RandomFlicker_Value04}
  n59 = Sine(None=Multiply(A=NamedRerouteUsage(), B=Multiply(A=@n57, B=@n58)))
  n60 = Time()
  n61 = ScalarParameter() {defaultValue=0.05, group=RandomFloatFlicker, parameterName=RandomFlicker_Value05}
  n62 = Sine(None=Multiply(A=NamedRerouteUsage(), B=Multiply(A=@n60, B=@n61)))
  n63 = Time()
  n64 = ScalarParameter() {defaultValue=0.264, group=RandomFloatFlicker, parameterName=RandomFlicker_Value06}
  n65 = Sine(None=Multiply(A=NamedRerouteUsage(), B=Multiply(A=@n63, B=@n64)))
  n66 = Saturate(None=LinearInterpolate(A=Add(A=@n50, B=@n53), B=Subtract(A=@n56, B=@n59), Alpha=Multiply(A=@n62, B=@n65)))
  n67 = Multiply(A=@n66, B=NamedRerouteUsage()) {constB=0.4}
  n68 = Sine(None=Multiply(A=Time(), B=Add(A=@n47, B=@n67)) {constB=0.2})
  NamedRerouteDeclaration(None=@n68) {name=Flicker, nodeColor=<rgba 0.4062 1 0 1>}
  OneMinus(None=@n68)
  n69 = Time()
  Reroute(None=Frac(None=Multiply(A=@n69) {constB=2}))
  Rotator(Coordinate=@n12, Time=Constant() {r=3.333})
  Step(Y=Power(Base=@n9.R) {constExponent=0.1}) {constX=0.5}
  Subtract(A=Multiply(A=@n66) {constB=2})
 ]
}
