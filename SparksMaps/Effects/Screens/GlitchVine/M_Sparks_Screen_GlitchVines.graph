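# Unlit Sparks stage screen material that draws glitching vines from a packed RGB texture over a patterned, glowing background.
# UVs support flip, tile, offset and mirror, with scrolling-texture distortion driven by sine-based glitch timing; three samples with clamped-sine offsets split color channels.
# Texture masks composite the vines over tiled gradients, time-stepped digital noise and bottom glow.
# Stage colors and an audio beat curve supply tint and reactivity, finished by a shared screen-details function.
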Material /SparksMaps/Effects/Screens/GlitchVine/M_Sparks_Screen_GlitchVines
graph {
 properties {
  bAutomaticallySetUsageInEditor=false
  bHasPixelAnimation=true
  bUsedWithNanite=true
  pixelDepthOffsetMode=PDOM_Legacy
  shadingModel=MSM_Unlit
  usedShadingModels=Unlit
 }
 nodes [
  Constant() {r=-0.01}
  Constant() {r=0.01}
  n1 = TextureSample(UVs=NamedRerouteUsage()) {
   samplerSource=SSM_Wrap_WorldGroupSettings
   samplerType=SAMPLERTYPE_LinearColor
   texture=&/T_GlitchVine_RGB
  }
  n2 = OneMinus(None=@n1.G)
  n3 = Multiply(A=Time(), B=Constant() {r=2}) {constB=1.76}
  n4 = Multiply(A=Time(), B=Constant() {r=2.75}) {constB=1.76}
  n5 = AppendVector(A=Clamp(None=Sine(None=@n3)) {maxDefault=0.004, minDefault=-0.005}, B=Clamp(None=Sine(None=@n4)) {maxDefault=0.004, minDefault=-0.005})
  n6 = Constant() {r=0.75}
  n7 = TextureSample(UVs=Add(A=NamedRerouteUsage(), B=Multiply(A=@n5, B=@n6))) {
   samplerSource=SSM_Wrap_WorldGroupSettings
   samplerType=SAMPLERTYPE_LinearColor
   texture=&T_GlitchVine_RGB
  }
  n8 = Multiply(A=Time(), B=Constant() {r=3}) {constB=2.43}
  n9 = Multiply(A=Time(), B=Constant() {r=2.5}) {constB=2.43}
  n10 = AppendVector(A=Clamp(None=Sine(None=@n8)) {maxDefault=0.005, minDefault=-0.006}, B=Clamp(None=Sine(None=@n9)) {maxDefault=0.005, minDefault=-0.006})
  n11 = TextureSample(UVs=Add(A=NamedRerouteUsage(), B=Multiply(A=@n6, B=@n10))) {
   samplerSource=SSM_Wrap_WorldGroupSettings
   samplerType=SAMPLERTYPE_LinearColor
   texture=&T_GlitchVine_RGB
  }
  n12 = MaterialFunctionCall(R=@n1.B, G=@n7.B, B=@n11.B) {materialFunction=&AppendMany}
  n13 = Multiply(A=NamedRerouteUsage(), B=Multiply(A=Multiply(A=@n2, B=@n12.RGB), B=Constant() {r=0.25}))
  n14 = Multiply(
   A=MaterialFunctionCall(R=@n1.G, G=@n7.G, B=@n11.G) {materialFunction=&AppendMany}.RGB
   B=Constant() {r=0.25}
  )
  n15 = Multiply(A=OneMinus(None=@n1.G), B=OneMinus(None=@n1.B))
  n16 = TextureCoordinate() {uTiling=8, vTiling=8}
  n17 = MaterialFunctionCall(UV Channel=Rotator(Coordinate=Frac(None=@n16), Time=Constant() {r=4})) {materialFunction=&LinearGradient}
  n18 = TextureCoordinate()
  n19 = OneMinus(None=@n18)
  n20 = TextureCoordinate() {uTiling=0.5, vTiling=2}
  n21 = AppendVector(A=Constant() {r=50}, B=Constant() {r=0.1})
  n22 = Constant() {r=3}
  n23 = Floor(None=Multiply(A=Time(), B=@n22))
  n24 = TextureSample(UVs=Panner(Coordinate=Multiply(A=@n20, B=@n21), Time=Divide(A=@n23, B=@n22)) {speedX=1, speedY=1}) {samplerType=SAMPLERTYPE_LinearColor, texture=/Game/Global/Textures/Default/Noise/T_Atlas_DigitalNoise_Small}
  n25 = LinearInterpolate(
   A=MaterialFunctionCall(Hue Shift Percentage=Multiply(A=@n17.UGradient) {constB=0.1}, Texture=NamedRerouteUsage()) {materialFunction=&HueShift}.Result
   Alpha=MaterialFunctionCall(UV Channel=LinearInterpolate(A=@n19, B=@n18, Alpha=@n24.R)) {materialFunction=&LinearGradient}.UGradient
  ) {constB=0}
  n26 = Constant() {r=0.1}
  n27 = Frac(None=TextureCoordinate() {uTiling=6, vTiling=6})
  n28 = Add(
   A=Power(Base=MaterialFunctionCall(UV Channel=@n27) {materialFunction=&LinearGradient}.VGradient) {constExponent=4}
   B=Constant() {r=0.15}
  ) {constB=0.025}
  n29 = NamedRerouteUsage()
  n30 = TextureCoordinate()
  n31 = Sine(None=Multiply(A=MaterialFunctionCall(UV Channel=@n30) {materialFunction=&LinearGradient}.UGradient) {constB=0.5})
  n32 = Saturate(None=Power(Base=OneMinus(None=@n31)) {constExponent=1})
  n33 = Add(
   A=Multiply(A=Multiply(A=@n25, B=@n26) {constB=0.05}, B=Saturate(None=@n28))
   B=Multiply(A=Multiply(A=@n29, B=@n32), B=Constant() {r=0.1}) {constB=0.05}
  )
  n34 = Multiply(A=ComponentMask(None=TextureCoordinate()) {r=true}) {constB=75}
  n35 = TextureCoordinate() {uTiling=0.5, vTiling=2}
  n36 = AppendVector(A=Constant() {r=0.25}, B=Constant() {r=1})
  n37 = Constant() {r=3}
  n38 = Floor(None=Multiply(A=Time(), B=@n37))
  n39 = TextureSample(UVs=Panner(Coordinate=Multiply(A=@n35, B=@n36), Time=Divide(A=@n38, B=@n37)) {speedX=1, speedY=1}) {samplerType=SAMPLERTYPE_LinearColor, texture=&T_Atlas_DigitalNoise_Small}
  n40 = Add(A=NamedRerouteUsage()) {constB=0.1}
  n41 = Multiply(A=Saturate(None=Sine(None=@n34)), B=Multiply(A=Multiply(A=@n39.R) {constB=0.05}, B=@n40)) {constB=0.1}
  n42 = ComponentMask(None=TextureCoordinate()) {g=true}
  n43 = ScalarParameter() {defaultValue=10, parameterName=Bot Glow Pow}
  n44 = Multiply(
   A=Multiply(A=Power(Base=@n42, Exp=@n43) {constExponent=10}, B=ScalarParameter() {defaultValue=2, parameterName=Bot Glow Mult}) {constB=0.1}
   B=@n40
  )
  n45 = Multiply(A=NamedRerouteUsage(), B=Add(A=Add(A=@n33, B=@n41), B=Saturate(None=@n44)))
  MaterialFunctionCall!(Emissive=Add(A=Add(A=@n13, B=@n14), B=Multiply(A=@n15, B=@n45)), Scanlines Tile) {materialFunction=/SparksMaps/Effects/Screens/Materials/MF/MF_Sparks_Screen_Details}
  NamedRerouteDeclaration(None=MaterialFunctionCall!(CustomUPosition) {materialFunction=/SparksStageShow/Systems/StageManager/FXManager/MaterialFunctions/MF_MojoStageColors}.Primary) {name=Primary Color, nodeColor=<rgba 0 0 0 1>}
  n46 = CurveAtlasRowParameter(CurveTime=CollectionParameter() {collection=/FMAudioReactiveData/AudioAnalysis/MPC_Sparks_Audio_Parameters, parameterName=BeatLifetime}) {
   atlas=/SparksMaps/Effects/Screens/Curves/CurveAtlas_Audio_Beat_A
   curve=/SparksMaps/Effects/Screens/Curves/Curve_Audio_Beat_Remap_001
   parameterName=ReactivityCurve
  }
  NamedRerouteDeclaration(None=MaterialFunctionCall(Input=@n46, Input Low=Constant(), Input High=Constant() {r=1}, Target Low=Constant() {r=0.5}, Target High=Constant() {r=1}) {materialFunction=&RemapValueRange}.Result) {name=Reactivity, nodeColor=<rgba 0 0 0 1>}
  n47 = TextureCoordinate()
  n48 = ComponentMask(None=@n47) {r=true}
  n49 = LinearInterpolate(A=@n48, B=OneMinus(None=@n48), Alpha=ScalarParameter() {parameterName=FlipX})
  n50 = ComponentMask(None=@n47) {g=true}
  n51 = ScalarParameter() {defaultValue=1, parameterName=XTile}
  n52 = ScalarParameter() {defaultValue=1, parameterName=YTile}
  n53 = Add(
   A=Multiply(A=AppendVector(A=@n49, B=@n50), B=AppendVector(A=@n51, B=@n52))
   B=AppendVector(A=ScalarParameter() {parameterName=XPos}, B=ScalarParameter() {defaultValue=0.15, parameterName=YPos})
  )
  n54 = MaterialFunctionCall!(Repeat Count=Constant2Vector() {g=1, r=2}, texcoord=@n53) {materialFunction=/Game/Environments/Papaya/Materials/MaterialFunctions/MF_UVMirror}
  n55 = ScalarParameter() {parameterName=MirrorUV}
  n56 = NamedRerouteUsage()
  n57 = Constant() {r=0.1}
  n58 = Constant() {r=0.25}
  n59 = TextureSample(UVs=Panner(Coordinate=TextureCoordinate(), Speed=If(A=@n56, "A > B"=@n57, "A < B"=@n58)) {speedX=0.1}) {
   samplerSource=SSM_Wrap_WorldGroupSettings
   samplerType=SAMPLERTYPE_LinearColor
   texture=&T_GlitchVine_RGB
  }
  n60 = If(A=NamedRerouteUsage(), "A > B"=Constant() {r=0.005}, "A < B"=Constant() {r=0.05})
  NamedRerouteDeclaration(None=Add(A=LinearInterpolate(A=@n53, B=@n54.Result, Alpha=@n55), B=Multiply(A=@n59.R, B=@n60) {constB=0.02})) {name=UVs, nodeColor=<rgba 1 0.2578 0 1>}
  n61 = Add(A=Time(), B=ScalarParameter() {parameterName=GlitchTimeOffset})
  n62 = ScalarParameter() {defaultValue=1, parameterName=Glitch Anim Amount}
  n63 = Multiply(A=Multiply(A=@n61) {constB=0.5}, B=@n62)
  n64 = Multiply(A=Multiply(A=@n61) {constB=0.376}, B=@n62)
  n65 = Multiply(A=Multiply(A=@n61) {constB=0.4}, B=@n62)
  NamedRerouteDeclaration(None=LinearInterpolate(A=Sine(None=@n63), B=Sine(None=@n64), Alpha=Sine(None=@n65))) {name=Glitch Time, nodeColor=<rgba 0 0.1094 1 1>}
  n66 = Time()
  Reroute(None=Frac(None=Multiply(A=@n66) {constB=2}))
  VectorParameter() {group=ScreenThemeParams, parameterName=TestColorA, defaultValue=<rgba 0.3078 1 0.1018 1>}
 ]
}
